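# Niagara system for the SheerWill character body's glowing eyes, with sprite glows, lights and mesh trails.
# EyesGlow and TrailMain each burst two persistent particles, partitioned per eye and pinned to the skeletal mesh with mirrored offsets.
# User.VariantColorIndex picks Blot, Citadel or Froth colors and intensities through scratch dynamic inputs.
# User.bisHeadHidden zeroes glow alpha, light radius and trail size; glow alpha also varies by platform, and trails brighten in the front end.
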NiagaraSystem /BRCosmetics/Characters/Player/Male/Medium/Bodies/M_MED_SheerWill_Blot/FX/NS_SheerWill_EyesGlow
dependencies {
 usedDataInterfaces [
  &NiagaraDataInterfaceArrayFloat3
  &NiagaraDataInterfacePlatformSet
  &NiagaraDataInterfaceArrayFloat
  &NiagaraDataInterfaceSkeletalMesh
 ]
 usedDynamicInputs [
  &self:ScratchDynamicInput_01
  &self:ScratchDynamicInput_02
  &self:ScratchDynamicInput_03
 ]
 usedModules [
  &self:SystemSpawnScript.SystemScriptSource.SystemScriptGraph.NiagaraNodeAssignment_0.SetVariables_#1
  &SystemState
  &self:SystemSpawnScript.SystemScriptSource.SystemScriptGraph.NiagaraNodeAssignment_1.SetVariables_#2
  &self:SystemSpawnScript.SystemScriptSource.SystemScriptGraph.NiagaraNodeAssignment_2.SetVariables_#3
  &self:SystemSpawnScript.SystemScriptSource.SystemScriptGraph.NiagaraNodeAssignment_3.SetVariables_#4
 ]
}
emitters {
 EyesGlow {
  emitterUpdateScript [
   &EmitterState(Life Cycle Mode=::NewEnumerator1, Loop Behavior=::NewEnumerator1, Loop Duration Mode=::NewEnumerator1, UseLoopDelay=false)
   &SpawnBurst_Instantaneous(Spawn Count=2, Use Spawn Probability=false, UseLoopCountLimit=false)
  ]
  particleSpawnScript [
   &PartitionParticles()
   &InitializeParticle(Color Mode=::NewEnumerator0, Sprite Size Mode=::NewEnumerator3, UsePositionOffset=false, Write Material Random=true)
   &CameraOffset(Camera Offset Amount=2)
   &DynamicMaterialParameters(
    Index 0 Param 1=3.3
    Index 0 Param 2=5
    Module.Index0Float0Write=true
    Module.Index0Float1Write=true
    Module.Index0Float2Write=true
    Module.Index0Float3Write=true
   )
  ]
  particleUpdateScript [
   &ParticleState(Kill Particles When Lifetime Has Elapsed=false)
   &Color(Color=&self:ScratchDynamicInput_04(
     BlotColor:LinearColor=<rgba 2 0 0.0033 1>
     CitadelColor:LinearColor=<rgba 0 0.6667 2 1>
     FrothColor:LinearColor=<rgba 0 0 0 1>
     VariantColorIndex:NiagaraInt32=&MakeIntFromFloat(Float=$User.VariantColorIndex)
    ))
   &SkeletalMeshLocation(
    Bone Sampling Mode=::NewEnumerator4
    MaskSpawnGroup=false
    Sampled Position Offset=&SelectVectorFromArray(
     Direct Array Index=$Particles.PartitionParticles.Partition
     Vector Selection Array=DataInterface() {propertyValues {FloatData [<4 9 3>, <4 9 -3>]}}
    )
    Skeletal Mesh=$System.Skeletal Mesh Reader
   )
   &SkeletalMeshLocation(
    Bone Sampling Mode=::NewEnumerator4
    MaskSpawnGroup=false
    Position Sampling=::NewEnumerator2
    Sampled Position Offset=&SelectVectorFromArray(
     Direct Array Index=$Particles.PartitionParticles.Partition
     Vector Selection Array=DataInterface() {propertyValues {FloatData [<4 13 5>, <4 13 -5>]}}
    )
    Skeletal Mesh=$System.Skeletal Mesh Reader
   ) {module=SkeletalMeshLocation001}
   &Light_Attributes(
    Color=1
    Color Mode=0
    Enabled=2
    Falloff Exponent=2
    Light Color:LinearColor=&self:ScratchDynamicInput_05(
     BlotColor:LinearColor=<rgba 20 0 0 1>
     CitadelColor:LinearColor=<rgba 0 6.6667 20 1>
     FrothColor:LinearColor=<rgba 0 0 0 1>
     VariantColorIndex:NiagaraInt32=&MakeIntFromFloat(Float=$User.VariantColorIndex)
    )
    Light Exponent=3
    Light Radius:NiagaraFloat=&Multiply_Float(A=10, B=&MakeCustomFloatFromBool(Bool=$User.bisHeadHidden, True Float=0, False Float=1))
    Radius=2
   )
   &ScaleColor(
    Color Value To Scale=$Particles.Color
    Scale Alpha=&Multiply_Float(
     A=&FN_PickFloatDependOnPlatform(TargetPlatformActive=1, TargetPlatformNoActive=0.1)
     B=&MakeCustomFloatFromBool(Bool=$User.bisHeadHidden, True Float=0, False Float=1)
    )
    ScaleA=true
    ScaleRGB=true
   )
  ]
  rendererClasses [&NiagaraSpriteRendererProperties, &NiagaraLightRendererProperties]
  renderers [
   {rendererClass=&NiagaraSpriteRendererProperties}
   {rendererClass=&NiagaraLightRendererProperties, rendererIndex=1}
  ]
 }
 TrailMain {
  emitterSpawnScript [/SetVariables(Emitter.Des:NiagaraFloat=&self:ScratchDynamicInput_08(
     Blot:NiagaraFloat=0.5
     Citadel:NiagaraFloat=0.5
     Froth:NiagaraFloat=1
     VariantColorIndex:NiagaraInt32=&MakeIntFromFloat(Float=$User.VariantColorIndex)
    )) {isSetParameters=true}]
  emitterUpdateScript [
   &EmitterState(Life Cycle Mode=::NewEnumerator1, Loop Behavior=::NewEnumerator1, Loop Duration Mode=::NewEnumerator1, UseLoopDelay=false)
   &SpawnBurst_Instantaneous(Spawn Count=2, Use Spawn Probability=false, UseLoopCountLimit=false)
  ]
  particleSpawnScript [
   &PartitionParticles()
   &InitializeParticle(
    Color=&self:ScratchDynamicInput_05(
     BlotColor:LinearColor=<rgba 1 1 1 1>
     CitadelColor:LinearColor=<rgba 1 1 1 1>
     FrothColor:LinearColor=<rgba 1 1 1 1.5>
     VariantColorIndex:NiagaraInt32=&MakeIntFromFloat(Float=$User.VariantColorIndex)
    )
    Mesh Scale Mode=::NewEnumerator3
    Mesh Uniform Scale=6
    UsePositionOffset=false
    Write Material Random=true
   )
   &InitialMeshOrientation(
    Mesh Orientation Mode=::NewEnumerator7
    Rotation=&SelectVectorFromArray(
     Direct Array Index=$Particles.PartitionParticles.Partition
     Vector Selection Array=DataInterface() {propertyValues {FloatData [<0 0.25 0>, <0 -0.25 0>]}}
    )
    Use Rotation Vector=true
   )
   &CameraOffset(Camera Offset Amount=3)
   &InitialMeshOrientation(
    Mesh Orientation Mode=::NewEnumerator7
    Rotation=&SelectVectorFromArray(
     Direct Array Index=$Particles.PartitionParticles.Partition
     Vector Selection Array=DataInterface() {propertyValues {FloatData [<-0.07 0 0>, <0.07 0 0>]}}
    )
    Use Rotation Vector=true
   ) {module=InitialMeshOrientation001}
   &InitialMeshOrientation(Mesh Orientation Mode=::NewEnumerator7, Rotation=<0 0 0.15>, Use Rotation Vector=true) {module=InitialMeshOrientation002}
  ]
  particleUpdateScript [
   &DynamicMaterialParameters(
    Index 0 Param 1=0.23
    Index 0 Param 2=&SelectFloatFromArray(Direct Array Index=$Particles.PartitionParticles.Partition, Float Selection Array=DataInterface() {propertyValues {FloatData [0.3,0.6]}})
    Module.Index0Float0Write=true
    Module.Index0Float1Write=true
    Module.Index0Float2Write=true
    Module.Index0Float3Write=true
   )
   &ParticleState(Kill Particles When Lifetime Has Elapsed=false)
   /SetVariables(Particles.SizeAll:NiagaraFloat=&Multiply_Float(A=0.036, B=&MakeCustomFloatFromBool(Bool=$User.bisHeadHidden, True Float=0, False Float=1))) {isSetParameters=true}
   &ScaleMeshSize(Scale Factor=&Multiply_VectorByFloat(Vector=<1 1 1>, Float=$Particles.SizeAll))
   &ScaleColor(
    Scale RGB=&Multiply_VectorByFloat(Vector=<1 1 1>, Float=&FN_FrontEndFloatModifier(Gameplay=1, Front End=1.5))
    ScaleA=true
    ScaleRGB=true
   )
   &FN_Mesh_Attributes(
    Color=&MakeLinearColorFromVectorAndFloat(
     "Vector (RGB)"=&Multiply_Vector(
      A=&MakeVectorFromLinearColorRGB(Color=&self:ScratchDynamicInput_05(
        BlotColor:LinearColor=<rgba 1 0.0146 0 0.4>
        CitadelColor:LinearColor=<rgba 0 0.3333 1 0.1>
        FrothColor:LinearColor=<rgba 0 0 0 1>
        VariantColorIndex:NiagaraInt32=&MakeIntFromFloat(Float=$User.VariantColorIndex)
       ))
      B=<1 1 1>
     )
     "Float (Alpha)"=&Multiply_Float(
      A=0.2
      B=&self:ScratchDynamicInput_06(
       Blot:NiagaraFloat=1
       Citadel:NiagaraFloat=2
       Froth:NiagaraFloat=2
       VariantColorIndex:NiagaraInt32=&MakeIntFromFloat(Float=$User.VariantColorIndex)
      )
     )
    )
    Create Color Attribute=true
    Module Output=::NewEnumerator1
   )
   &SkeletalMeshLocation(
    Bone / Socket Index=$Particles.PartitionParticles.Partition
    Bone Sampling Mode=::NewEnumerator4
    MaskSpawnGroup=false
    Sampled Position Offset=&SelectVectorFromArray(
     Direct Array Index=$Particles.PartitionParticles.Partition
     Vector Selection Array=DataInterface() {propertyValues {FloatData [<3.5 10 -4>, <3.5 10 4>]}}
    )
    Skeletal Mesh=$System.Skeletal Mesh Reader
   )
  ]
  rendererClasses [&NiagaraMeshRendererProperties]
  renderers [
   {rendererClass=&NiagaraMeshRendererProperties, rendererIndex=0}
   {rendererClass=&NiagaraMeshRendererProperties, rendererIndex=1}
  ]
 }
}
summary {
 systemName=NS_SheerWill_EyesGlow
 emitters [
  {emitterName=EyesGlow, rendererClasses [&NiagaraSpriteRendererProperties, &NiagaraLightRendererProperties]}
  {emitterName=TrailMain, rendererClasses [&NiagaraMeshRendererProperties]}
 ]
 userVariables [
  User.bIsFrontEnd {type=NiagaraBool}
  User.bIsFrontEndPreview {type=NiagaraBool}
  User.bisHeadHidden {type=NiagaraBool}
  User.Variant {type=NiagaraFloat, defaultValue=1, description=""}
  User.VariantColorIndex {type=NiagaraFloat, defaultValue=0, description=""}
 ]
}
