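# Physics asset for the M_MED_SheerWill_Blot male player body cosmetic, set up for dangling front and back loincloth physics.
# Each loincloth is a four-link capsule chain on a base bone constrained to the pelvis, with tight angular limits and less swing in front.
# Pelvis, thigh, calf and foot capsules plus a flat Ground_Joint box provide body and floor shapes; leg and base-to-pelvis joints rotate freely.
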PhysicsAsset /BRCosmetics/Characters/Player/Male/Medium/Bodies/M_MED_SheerWill_Blot/Meshes/M_MED_SheerWill_Blot_Physics
bodies {
 Ground_Joint [{shapeType=Box, extentX=172.5632, extentY=247.21, extentZ=9.8513, shapeName=Ground_Joint_box, center=<0 0 -4.1868>, rotation=<rot 0 -90 0>}]
 base_loincloth_back [{shapeType=Capsule, length=5, radius=7, shapeName=base_loincloth_back_capsule, rotation=<rot 0 0 90>}]
 base_loincloth_front [{shapeType=Capsule, length=0.505, radius=0.505, shapeName=base_loincloth_front_capsule, rotation=<rot 0 0 90>}]
 calf_l [{shapeType=Capsule, length=45.1738, radius=7.9838, shapeName=calf_l_capsule, center=<-24.5444 -2.0769 -0.1995>, rotation=<rot 90 0 0>}]
 calf_r [{shapeType=Capsule, length=43.468, radius=7.7653, shapeName=calf_r_capsule, center=<23.0872 2.2437 0>, rotation=<rot 90 0 0>}]
 dyn_loincloth_back_1 [{shapeType=Capsule, length=9.2063, radius=7, shapeName=dyn_loincloth_back_1_capsule, center=<7.4443 0 4.7486>, rotation=<rot 0 0 90>}]
 dyn_loincloth_back_2 [{shapeType=Capsule, length=7.7997, radius=7, shapeName=dyn_loincloth_back_2_capsule, center=<0 0 4.1864>, rotation=<rot 0 0 90>}]
 dyn_loincloth_back_3 [{shapeType=Capsule, length=5, radius=7, shapeName=dyn_loincloth_back_3_capsule, center=<0 0 3.837>, rotation=<rot 0 0 90>}]
 dyn_loincloth_back_4 [{shapeType=Capsule, length=5, radius=7, shapeName=dyn_loincloth_back_4_capsule, center=<0 0 3.7031>, rotation=<rot 0 0 90>}]
 dyn_loincloth_front_1 [
  {shapeType=Capsule, length=4.9954, radius=4.7514, shapeName=dyn_loincloth_front_1_capsule, center=<7.9355 0 -3.4426>, rotation=<rot 0 0 90>}
  {shapeType=Capsule, length=6, radius=5, shapeName=dyn_loincloth_front_1_capsule, center=<0 0 -1.9175>, rotation=<rot 0 0 90>}
 ]
 dyn_loincloth_front_2 [{shapeType=Capsule, length=4.9954, radius=4.7514, shapeName=dyn_loincloth_front_2_capsule, center=<4 0 -3.3121>, rotation=<rot 0 0 90>}]
 dyn_loincloth_front_3 [{shapeType=Capsule, length=4.9954, radius=4.7514, shapeName=dyn_loincloth_front_3_capsule, center=<4 0 -3.1376>, rotation=<rot 0 0 90>}]
 dyn_loincloth_front_4 [{shapeType=Capsule, length=4.9954, radius=4.7514, shapeName=dyn_loincloth_front_4_capsule, center=<4 0 -2.6385>, rotation=<rot 0 0 90>}]
 foot_l [{shapeType=Capsule, length=14.7137, radius=9.1019, shapeName=foot_l_capsule, center=<-2.6477 6.3546 -0.017>, rotation=<rot 0 9.892 -89.1526>}]
 foot_r [{shapeType=Capsule, length=14.7128, radius=9.0995, shapeName=foot_r_capsule, center=<2.6442 -6.3547 0.0131>, rotation=<rot 0 9.8643 -89.0945>}]
 pelvis [{shapeType=Capsule, length=11.7038, radius=10.797, center=<-1.8139 0 0>}]
 thigh_l [{shapeType=Capsule, length=25.5, radius=9, shapeName=thigh_l_capsule, center=<-15.5961 0 0>, rotation=<rot 90 0 0>}]
 thigh_r [{shapeType=Capsule, length=25.5, radius=9, shapeName=thigh_r_capsule, center=<14.5076 0.4318 -1.2679>, rotation=<rot 90 0 0>}]
}
constraints [
 {
  bone1Name=base_loincloth_back
  bone2Name=pelvis
  swing1LimitDegrees=45
  swing1Motion=Free
  swing2LimitDegrees=45
  swing2Motion=Free
  twistLimitDegrees=45
  twistMotion=Free
 }
 {bone1Name=dyn_loincloth_back_1, bone2Name=base_loincloth_back, swing1LimitDegrees=10, swing2LimitDegrees=35, twistLimitDegrees=1}
 {bone1Name=dyn_loincloth_back_2, bone2Name=dyn_loincloth_back_1, swing1LimitDegrees=10, swing2LimitDegrees=35, twistLimitDegrees=1}
 {bone1Name=dyn_loincloth_back_3, bone2Name=dyn_loincloth_back_2, swing1LimitDegrees=10, swing2LimitDegrees=35, twistLimitDegrees=1}
 {bone1Name=dyn_loincloth_back_4, bone2Name=dyn_loincloth_back_3, swing1LimitDegrees=10, swing2LimitDegrees=35, twistLimitDegrees=1}
 {
  bone1Name=thigh_l
  bone2Name=pelvis
  swing1LimitDegrees=45
  swing1Motion=Free
  swing2LimitDegrees=45
  swing2Motion=Free
  twistLimitDegrees=45
  twistMotion=Free
 }
 {
  bone1Name=thigh_r
  bone2Name=pelvis
  swing1LimitDegrees=45
  swing1Motion=Free
  swing2LimitDegrees=45
  swing2Motion=Free
  twistLimitDegrees=45
  twistMotion=Free
 }
 {
  bone1Name=calf_l
  bone2Name=thigh_l
  swing1LimitDegrees=45
  swing1Motion=Free
  swing2LimitDegrees=45
  swing2Motion=Free
  twistLimitDegrees=45
  twistMotion=Free
 }
 {
  bone1Name=calf_r
  bone2Name=thigh_r
  swing1LimitDegrees=45
  swing1Motion=Free
  swing2LimitDegrees=45
  swing2Motion=Free
  twistLimitDegrees=45
  twistMotion=Free
 }
 {
  bone1Name=base_loincloth_front
  bone2Name=pelvis
  swing1LimitDegrees=45
  swing1Motion=Free
  swing2LimitDegrees=45
  swing2Motion=Free
  twistLimitDegrees=45
  twistMotion=Free
 }
 {
  bone1Name=dyn_loincloth_front_1
  bone2Name=base_loincloth_front
  swing1LimitDegrees=5
  swing2LimitDegrees=7
  twistLimitDegrees=1
  twistMotion=Locked
 }
 {bone1Name=dyn_loincloth_front_2, bone2Name=dyn_loincloth_front_1, swing1LimitDegrees=5, swing2LimitDegrees=15, twistLimitDegrees=1}
 {bone1Name=dyn_loincloth_front_3, bone2Name=dyn_loincloth_front_2, swing1LimitDegrees=5, swing2LimitDegrees=15, twistLimitDegrees=1}
 {bone1Name=dyn_loincloth_front_4, bone2Name=dyn_loincloth_front_3, swing1LimitDegrees=5, swing2LimitDegrees=15, twistLimitDegrees=1}
 {
  bone1Name=foot_l
  bone2Name=calf_l
  swing1LimitDegrees=45
  swing1Motion=Free
  swing2LimitDegrees=45
  swing2Motion=Free
  twistLimitDegrees=45
  twistMotion=Free
 }
 {
  bone1Name=foot_r
  bone2Name=calf_r
  swing1LimitDegrees=45
  swing1Motion=Free
  swing2LimitDegrees=45
  swing2Motion=Free
  twistLimitDegrees=45
  twistMotion=Free
 }
]
