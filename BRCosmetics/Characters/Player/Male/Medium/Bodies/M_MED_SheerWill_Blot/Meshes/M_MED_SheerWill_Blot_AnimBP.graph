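# Post-process style AnimBP for the SheerWill Blot body cosmetic, adding secondary motion and floating shards to the parent mesh's copied pose.
# Control Rig feeds per-shard AnimDynamics, a pelvis-space RigidBody, then loincloth and head/neck bone offsets.
# Dyn Shards computes three time-based sine/cosine vectors that additively bob the shard bones.
# Head_Mesh supplies facial bones, and a custom-idle DefaultSlot blends over neck and head before output.
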AnimBlueprint /BRCosmetics/Characters/Player/Male/Medium/Bodies/M_MED_SheerWill_Blot/Meshes/M_MED_SheerWill_Blot_AnimBP
target_skeleton=&/M_MED_SheerWill_Blot_Skeleton
graphs [
 AnimGraph {nodes [
   n1 = Anim Dynamics - Bone: dyn_upper_shard_l(Alpha=0.1, ExternalForce="0,0,0") {
    angularDampingOverride=1
    angularSpringConstant=20000
    bAngularSpring=true
    bOverrideAngularDamping=true
    bUseGravityOverride=true
    gravityScale=0.35
    linearSpringConstant=1
    simulationSpace=RootRelative
    worldSpaceMinimumScale=0.1
    boundBone {boneName=dyn_upper_shard_l}
    componentAppliedLinearAccClamp=<0 0 0>
    physicsBodyDefinitions [{
      boundBone {boneName=dyn_upper_shard_l}
      boxExtents=<5 5 5>
      constraintSetup {angularLimitsMax=<15 15 15>, angularLimitsMin=<-15 -15 -15>, angularTarget=<1 0 0>, linearAxesMax=<5 5 5>, linearAxesMin=<-5 -5 -5>}
      localJointOffset=<0 0 0>
     }]
    relativeSpaceBone {boneName=spine_03}
    simSpaceSettings {simSpaceAngularAlpha=0.1}
   }
   n2 = Anim Dynamics - Bone: dyn_upper_shard_r(ComponentPose=@n1, Alpha=0.1, ExternalForce="0,0,0") {
    angularDampingOverride=1
    angularSpringConstant=20000
    bAngularSpring=true
    bOverrideAngularDamping=true
    bUseGravityOverride=true
    gravityScale=0.35
    linearSpringConstant=1
    simulationSpace=RootRelative
    worldSpaceMinimumScale=0.1
    boundBone {boneName=dyn_upper_shard_r}
    componentAppliedLinearAccClamp=<0 0 0>
    physicsBodyDefinitions [{
      boundBone {boneName=dyn_upper_shard_r}
      boxExtents=<5 5 5>
      constraintSetup {angularLimitsMax=<15 15 15>, angularLimitsMin=<-15 -15 -15>, angularTarget=<1 0 0>, linearAxesMax=<5 5 5>, linearAxesMin=<-5 -5 -5>}
      localJointOffset=<0 0 0>
     }]
    relativeSpaceBone {boneName=spine_03}
    simSpaceSettings {simSpaceAngularAlpha=0.1}
   }
   n3 = Anim Dynamics - Bone: dyn_shard_l_1(ComponentPose=@n2, Alpha=0.1, ExternalForce="0,0,0") {
    angularDampingOverride=1
    angularSpringConstant=20000
    bAngularSpring=true
    bOverrideAngularDamping=true
    bUseGravityOverride=true
    gravityScale=0.35
    linearSpringConstant=1
    simulationSpace=RootRelative
    worldSpaceMinimumScale=0.1
    boundBone {boneName=dyn_shard_l_1}
    componentAppliedLinearAccClamp=<0 0 0>
    physicsBodyDefinitions [{
      boundBone {boneName=dyn_shard_l_1}
      boxExtents=<5 5 5>
      constraintSetup {angularLimitsMax=<15 15 15>, angularLimitsMin=<-15 -15 -15>, angularTarget=<1 0 0>, linearAxesMax=<5 5 5>, linearAxesMin=<-5 -5 -5>}
      localJointOffset=<0 0 0>
     }]
    relativeSpaceBone {boneName=spine_03}
    simSpaceSettings {simSpaceAngularAlpha=0.1}
   }
   n4 = Anim Dynamics - Bone: dyn_shard_l_2(ComponentPose=@n3, Alpha=0.1, ExternalForce="0,0,0") {
    angularDampingOverride=1
    angularSpringConstant=20000
    bAngularSpring=true
    bOverrideAngularDamping=true
    bUseGravityOverride=true
    gravityScale=0.35
    linearSpringConstant=1
    simulationSpace=RootRelative
    worldSpaceMinimumScale=0.1
    boundBone {boneName=dyn_shard_l_2}
    componentAppliedLinearAccClamp=<0 0 0>
    physicsBodyDefinitions [{
      boundBone {boneName=dyn_shard_l_2}
      boxExtents=<5 5 5>
      constraintSetup {angularLimitsMax=<15 15 15>, angularLimitsMin=<-15 -15 -15>, angularTarget=<1 0 0>, linearAxesMax=<5 5 5>, linearAxesMin=<-5 -5 -5>}
      localJointOffset=<0 0 0>
     }]
    relativeSpaceBone {boneName=spine_03}
    simSpaceSettings {simSpaceAngularAlpha=0.1}
   }
   n5 = Anim Dynamics - Bone: dyn_shard_l_3(ComponentPose=@n4, Alpha=0.1, ExternalForce="0,0,0") {
    angularDampingOverride=1
    angularSpringConstant=20000
    bAngularSpring=true
    bOverrideAngularDamping=true
    bUseGravityOverride=true
    gravityScale=0.35
    linearSpringConstant=1
    simulationSpace=RootRelative
    worldSpaceMinimumScale=0.1
    boundBone {boneName=dyn_shard_l_3}
    componentAppliedLinearAccClamp=<0 0 0>
    physicsBodyDefinitions [{
      boundBone {boneName=dyn_shard_l_3}
      boxExtents=<5 5 5>
      constraintSetup {angularLimitsMax=<15 15 15>, angularLimitsMin=<-15 -15 -15>, angularTarget=<1 0 0>, linearAxesMax=<5 5 5>, linearAxesMin=<-5 -5 -5>}
      localJointOffset=<0 0 0>
     }]
    relativeSpaceBone {boneName=spine_03}
    simSpaceSettings {simSpaceAngularAlpha=0.1}
   }
   n6 = Anim Dynamics - Bone: dyn_shard_r_3(ComponentPose=@n5, Alpha=0.1, ExternalForce="0,0,0") {
    angularDampingOverride=1
    angularSpringConstant=20000
    bAngularSpring=true
    bOverrideAngularDamping=true
    bUseGravityOverride=true
    gravityScale=0.35
    linearSpringConstant=1
    simulationSpace=RootRelative
    worldSpaceMinimumScale=0.1
    boundBone {boneName=dyn_shard_r_3}
    componentAppliedLinearAccClamp=<0 0 0>
    physicsBodyDefinitions [{
      boundBone {boneName=dyn_shard_r_3}
      boxExtents=<5 5 5>
      constraintSetup {angularLimitsMax=<15 15 15>, angularLimitsMin=<-15 -15 -15>, angularTarget=<1 0 0>, linearAxesMax=<5 5 5>, linearAxesMin=<-5 -5 -5>}
      localJointOffset=<0 0 0>
     }]
    relativeSpaceBone {boneName=spine_03}
    simSpaceSettings {simSpaceAngularAlpha=0.1}
   }
   n7 = Anim Dynamics - Bone: dyn_shard_r_4(ComponentPose=@n6, Alpha=0.1, ExternalForce="0,0,0") {
    angularDampingOverride=1
    angularSpringConstant=20000
    bAngularSpring=true
    bOverrideAngularDamping=true
    bUseGravityOverride=true
    gravityScale=0.35
    linearSpringConstant=1
    simulationSpace=RootRelative
    worldSpaceMinimumScale=0.1
    boundBone {boneName=dyn_shard_r_4}
    componentAppliedLinearAccClamp=<0 0 0>
    physicsBodyDefinitions [{
      boundBone {boneName=dyn_shard_r_4}
      boxExtents=<5 5 5>
      constraintSetup {angularLimitsMax=<15 15 15>, angularLimitsMin=<-15 -15 -15>, angularTarget=<1 0 0>, linearAxesMax=<5 5 5>, linearAxesMin=<-5 -5 -5>}
      localJointOffset=<0 0 0>
     }]
    relativeSpaceBone {boneName=spine_03}
    simSpaceSettings {simSpaceAngularAlpha=0.1}
   }
   n8 = Anim Dynamics - Bone: dyn_middle_shard_l(ComponentPose=@n7, Alpha=0.1, ExternalForce="0,0,0") {
    angularDampingOverride=1
    angularSpringConstant=20000
    bAngularSpring=true
    bOverrideAngularDamping=true
    bUseGravityOverride=true
    gravityScale=0.35
    linearSpringConstant=1
    simulationSpace=RootRelative
    worldSpaceMinimumScale=0.1
    boundBone {boneName=dyn_middle_shard_l}
    componentAppliedLinearAccClamp=<0 0 0>
    physicsBodyDefinitions [{
      boundBone {boneName=dyn_middle_shard_l}
      boxExtents=<5 5 5>
      constraintSetup {angularLimitsMax=<15 15 15>, angularLimitsMin=<-15 -15 -15>, angularTarget=<1 0 0>, linearAxesMax=<5 5 5>, linearAxesMin=<-5 -5 -5>}
      localJointOffset=<0 0 0>
     }]
    relativeSpaceBone {boneName=spine_03}
    simSpaceSettings {simSpaceAngularAlpha=0.1}
   }
   n9 = Anim Dynamics - Bone: dyn_middle_shard_r(ComponentPose=@n8, Alpha=0.1, ExternalForce="0,0,0") {
    angularDampingOverride=1
    angularSpringConstant=20000
    bAngularSpring=true
    bOverrideAngularDamping=true
    bUseGravityOverride=true
    gravityScale=0.35
    linearSpringConstant=1
    simulationSpace=RootRelative
    worldSpaceMinimumScale=0.1
    boundBone {boneName=dyn_middle_shard_r}
    componentAppliedLinearAccClamp=<0 0 0>
    physicsBodyDefinitions [{
      boundBone {boneName=dyn_middle_shard_r}
      boxExtents=<5 5 5>
      constraintSetup {angularLimitsMax=<15 15 15>, angularLimitsMin=<-15 -15 -15>, angularTarget=<1 0 0>, linearAxesMax=<5 5 5>, linearAxesMin=<-5 -5 -5>}
      localJointOffset=<0 0 0>
     }]
    relativeSpaceBone {boneName=spine_03}
    simSpaceSettings {simSpaceAngularAlpha=0.1}
   }
   n10 = Anim Dynamics - Bone: dyn_shard_arm_l_01(ComponentPose=@n9, Alpha=0.1, ExternalForce="0,0,0") {
    angularDampingOverride=1
    angularSpringConstant=20000
    bAngularSpring=true
    bOverrideAngularDamping=true
    bUseGravityOverride=true
    gravityScale=0.35
    linearSpringConstant=1
    simulationSpace=RootRelative
    worldSpaceMinimumScale=0.1
    boundBone {boneName=dyn_shard_arm_l_01}
    componentAppliedLinearAccClamp=<0 0 0>
    physicsBodyDefinitions [{
      boundBone {boneName=dyn_shard_arm_l_01}
      boxExtents=<5 5 5>
      constraintSetup {angularLimitsMax=<15 15 15>, angularLimitsMin=<-15 -15 -15>, angularTarget=<1 0 0>, linearAxesMax=<5 5 5>, linearAxesMin=<-5 -5 -5>}
      localJointOffset=<0 0 0>
     }]
    relativeSpaceBone {boneName=spine_03}
    simSpaceSettings {simSpaceAngularAlpha=0.1}
   }
   n11 = Anim Dynamics - Bone: dyn_shard_arm_r_01(ComponentPose=@n10, Alpha=0.1, ExternalForce="0,0,0") {
    angularDampingOverride=1
    angularSpringConstant=20000
    bAngularSpring=true
    bOverrideAngularDamping=true
    bUseGravityOverride=true
    gravityScale=0.35
    linearSpringConstant=1
    simulationSpace=RootRelative
    worldSpaceMinimumScale=0.1
    boundBone {boneName=dyn_shard_arm_r_01}
    componentAppliedLinearAccClamp=<0 0 0>
    physicsBodyDefinitions [{
      boundBone {boneName=dyn_shard_arm_r_01}
      boxExtents=<5 5 5>
      constraintSetup {angularLimitsMax=<15 15 15>, angularLimitsMin=<-15 -15 -15>, angularTarget=<1 0 0>, linearAxesMax=<5 5 5>, linearAxesMin=<-5 -5 -5>}
      localJointOffset=<0 0 0>
     }]
    relativeSpaceBone {boneName=spine_03}
    simSpaceSettings {simSpaceAngularAlpha=0.1}
   }
   n12 = Anim Dynamics - Bone: dyn_shard_bottom_m_01(ComponentPose=@n11, Alpha=0.1, ExternalForce="0,0,0") {
    angularDampingOverride=1
    angularSpringConstant=20000
    bAngularSpring=true
    bOverrideAngularDamping=true
    bUseGravityOverride=true
    gravityScale=0.35
    linearSpringConstant=1
    simulationSpace=RootRelative
    worldSpaceMinimumScale=0.1
    boundBone {boneName=dyn_shard_bottom_m_01}
    componentAppliedLinearAccClamp=<0 0 0>
    physicsBodyDefinitions [{
      boundBone {boneName=dyn_shard_bottom_m_01}
      boxExtents=<5 5 5>
      constraintSetup {angularLimitsMax=<15 15 15>, angularLimitsMin=<-15 -15 -15>, angularTarget=<1 0 0>, linearAxesMax=<5 5 5>, linearAxesMin=<-5 -5 -5>}
      localJointOffset=<0 0 0>
     }]
    relativeSpaceBone {boneName=spine_03}
    simSpaceSettings {simSpaceAngularAlpha=0.1}
   }
   n13 = Anim Dynamics - Bone: dyn_shard_bottom_r_01(ComponentPose=@n12, Alpha=0.1, ExternalForce="0,0,0") {
    angularDampingOverride=1
    angularSpringConstant=20000
    bAngularSpring=true
    bOverrideAngularDamping=true
    bUseGravityOverride=true
    gravityScale=0.35
    linearSpringConstant=1
    simulationSpace=RootRelative
    worldSpaceMinimumScale=0.1
    boundBone {boneName=dyn_shard_bottom_r_01}
    componentAppliedLinearAccClamp=<0 0 0>
    physicsBodyDefinitions [{
      boundBone {boneName=dyn_shard_bottom_r_01}
      boxExtents=<5 5 5>
      constraintSetup {angularLimitsMax=<15 15 15>, angularLimitsMin=<-15 -15 -15>, angularTarget=<1 0 0>, linearAxesMax=<5 5 5>, linearAxesMin=<-5 -5 -5>}
      localJointOffset=<0 0 0>
     }]
    relativeSpaceBone {boneName=spine_03}
    simSpaceSettings {simSpaceAngularAlpha=0.1}
   }
   Anim Dynamics - Bone: dyn_shard_bottom_l_01(ComponentPose=@n13, Alpha=0.1, ExternalForce="0,0,0") {
    angularDampingOverride=1
    angularSpringConstant=20000
    bAngularSpring=true
    bOverrideAngularDamping=true
    bUseGravityOverride=true
    gravityScale=0.35
    linearSpringConstant=1
    simulationSpace=RootRelative
    worldSpaceMinimumScale=0.1
    boundBone {boneName=dyn_shard_bottom_l_01}
    componentAppliedLinearAccClamp=<0 0 0>
    physicsBodyDefinitions [{
      boundBone {boneName=dyn_shard_bottom_l_01}
      boxExtents=<5 5 5>
      constraintSetup {angularLimitsMax=<15 15 15>, angularLimitsMin=<-15 -15 -15>, angularTarget=<1 0 0>, linearAxesMax=<5 5 5>, linearAxesMin=<-5 -5 -5>}
      localJointOffset=<0 0 0>
     }]
    relativeSpaceBone {boneName=spine_03}
    simSpaceSettings {simSpaceAngularAlpha=0.1}
   }
   Dynamic Shards!(InPose:Pose Link Structure) -> (Pose:Pose Link Structure) {is=title AnimGraphNode_LinkedAnimLayer, layer=Dynamic Shards}
   "Transform (Modify) Bone - Bone: Collar"(Alpha=Get inGame() -> (inGame:double), Translation="0,0,0", Rotation="0,0,0", Scale="1,1,1") {rotationMode=BMM_Additive, rotationSpace=BCS_BoneSpace, boneToModify {boneName=Collar}}
   n14 = Layered blend per bone(
    BasePose=Use cached pose 'CACHE_NormalSetup'()
    BlendPoses_0=Slot 'DefaultSlot'(Source=Use cached pose 'Base Pose'())
    BlendWeights_0=Get Custom Idle Is Playing Float() -> (Custom Idle Is Playing Float:double)
   ) {layerSetup [{branchFilters [{boneName=neck_01}, {boneName=neck_02}, {boneName=head}]}]}
   Output Pose(Result=@n14) {layerGroup=None}
   n15 = Control Rig!(
    Source=Use cached pose 'Base Pose'()
    Alpha
    BaseballSliding:bool=Get bIsBaseballSliding()
    Sliding:bool=Get bIsKneeSliding()
    isCrouching:bool=Get bIsCrouching()
    Skydiving:bool=Get bIsSkyDiving()
    IsEmoting:bool=Get bIsPlayingEmote()
    inGame:bool="To Bool (Double)"!(Value:double=Get inGame()) -> (ReturnValue:bool)
   ) -> (Pose) {controlRigAssetReference {blueprintRigClass=&/M_MED_SheerWill_Blot_CtrlRig.M_MED_SheerWill_Blot_CtrlRig_C}}
   n16 = Anim Dynamics - Bone: dyn_shard_l_1(ComponentPose=Local To Component(LocalPose=@n15), Alpha=1, ExternalForce="0,0,0") {
    angularDampingOverride=1
    angularSpringConstant=20000
    bAngularSpring=true
    bOverrideAngularDamping=true
    bOverrideLinearDamping=true
    bUseGravityOverride=true
    gravityScale=0.35
    linearDampingOverride=1
    simulationSpace=RootRelative
    boundBone {boneName=dyn_shard_l_1}
    componentLinearAccScale=<1 1 1>
    componentLinearVelScale=<1 1 1>
    physicsBodyDefinitions [{
      boundBone {boneName=dyn_shard_l_1}
      boxExtents=<20 30 30>
      constraintSetup {angularLimitsMax=<5 5 5>, angularLimitsMin=<-5 -5 -5>, angularTarget=<1 0 0>, linearAxesMax=<1 1 1>, linearAxesMin=<-1 -1 -1>}
      localJointOffset=<2 0 0>
     }]
    relativeSpaceBone {boneName=spine_05}
    simSpaceSettings {simSpaceAngularAlpha=1}
   }
   n17 = Anim Dynamics - Bone: dyn_shard_l_2(ComponentPose=@n16, Alpha=1, ExternalForce="0,0,0") {
    angularDampingOverride=1
    angularSpringConstant=20000
    bAngularSpring=true
    bOverrideAngularDamping=true
    bOverrideLinearDamping=true
    bUseGravityOverride=true
    gravityScale=0.35
    linearDampingOverride=1
    simulationSpace=RootRelative
    boundBone {boneName=dyn_shard_l_2}
    componentLinearAccScale=<1 1 1>
    componentLinearVelScale=<1 1 1>
    physicsBodyDefinitions [{
      boundBone {boneName=dyn_shard_l_2}
      boxExtents=<20 30 30>
      constraintSetup {angularLimitsMax=<5 5 5>, angularLimitsMin=<-5 -5 -5>, angularTarget=<1 0 0>, linearAxesMax=<1 2 1>, linearAxesMin=<-1 -1 -1>}
      localJointOffset=<1 0 0>
     }]
    relativeSpaceBone {boneName=spine_05}
    simSpaceSettings {simSpaceAngularAlpha=1}
   }
   n18 = Anim Dynamics - Bone: dyn_shard_l_3(ComponentPose=@n17, Alpha=1, ExternalForce="0,0,0") {
    angularDampingOverride=1
    angularSpringConstant=20000
    bAngularSpring=true
    bOverrideAngularDamping=true
    bOverrideLinearDamping=true
    bUseGravityOverride=true
    gravityScale=0.35
    linearDampingOverride=1
    simulationSpace=RootRelative
    boundBone {boneName=dyn_shard_l_3}
    componentLinearAccScale=<1 1 1>
    componentLinearVelScale=<1 1 1>
    physicsBodyDefinitions [{
      boundBone {boneName=dyn_shard_l_3}
      boxExtents=<20 30 30>
      constraintSetup {angularLimitsMax=<5 5 5>, angularLimitsMin=<-5 -5 -5>, angularTarget=<1 0 0>, linearAxesMax=<1 1 1>, linearAxesMin=<-1 -1 -1>}
      localJointOffset=<4 0 0>
     }]
    relativeSpaceBone {boneName=spine_05}
    simSpaceSettings {simSpaceAngularAlpha=1}
   }
   n19 = Anim Dynamics - Bone: dyn_shard_r_3(ComponentPose=@n18, Alpha=1, ExternalForce="0,0,0") {
    angularDampingOverride=1
    angularSpringConstant=20000
    bAngularSpring=true
    bOverrideAngularDamping=true
    bOverrideLinearDamping=true
    bUseGravityOverride=true
    gravityScale=0.35
    linearDampingOverride=1
    simulationSpace=RootRelative
    boundBone {boneName=dyn_shard_r_3}
    componentLinearAccScale=<1 1 1>
    componentLinearVelScale=<1 1 1>
    physicsBodyDefinitions [{
      boundBone {boneName=dyn_shard_r_3}
      boxExtents=<20 30 30>
      constraintSetup {angularLimitsMax=<5 5 5>, angularLimitsMin=<-5 -5 -5>, angularTarget=<1 0 0>, linearAxesMax=<1 1 1>, linearAxesMin=<-1 -1 -1>}
      localJointOffset=<2 0 0>
     }]
    relativeSpaceBone {boneName=spine_05}
    simSpaceSettings {simSpaceAngularAlpha=1}
   }
   n20 = Anim Dynamics - Bone: dyn_shard_r_4(ComponentPose=@n19, Alpha=1, ExternalForce="0,0,0") {
    angularDampingOverride=1
    angularSpringConstant=20000
    bAngularSpring=true
    bOverrideAngularDamping=true
    bOverrideLinearDamping=true
    bUseGravityOverride=true
    gravityScale=0.35
    linearDampingOverride=1
    simulationSpace=RootRelative
    boundBone {boneName=dyn_shard_r_4}
    componentLinearAccScale=<1 1 1>
    componentLinearVelScale=<1 1 1>
    physicsBodyDefinitions [{
      boundBone {boneName=dyn_shard_r_4}
      boxExtents=<20 30 30>
      constraintSetup {angularLimitsMax=<5 5 5>, angularLimitsMin=<-5 -5 -5>, angularTarget=<1 0 0>, linearAxesMax=<1 1 1>, linearAxesMin=<-1 -1 -1>}
      localJointOffset=<4 0 0>
     }]
    relativeSpaceBone {boneName=spine_05}
    simSpaceSettings {simSpaceAngularAlpha=1}
   }
   n21 = Anim Dynamics - Bone: dyn_middle_shard_l(ComponentPose=@n20, Alpha=1, ExternalForce="0,0,0") {
    angularDampingOverride=1
    angularSpringConstant=20000
    bAngularSpring=true
    bOverrideAngularDamping=true
    bOverrideLinearDamping=true
    bUseGravityOverride=true
    gravityScale=0.35
    linearDampingOverride=1
    simulationSpace=RootRelative
    boundBone {boneName=dyn_middle_shard_l}
    componentLinearAccScale=<1 1 1>
    componentLinearVelScale=<1 1 1>
    physicsBodyDefinitions [{
      boundBone {boneName=dyn_middle_shard_l}
      boxExtents=<20 30 30>
      constraintSetup {angularLimitsMax=<5 5 5>, angularLimitsMin=<-5 -5 -5>, angularTarget=<1 0 0>, linearAxesMax=<1 1 1>, linearAxesMin=<-1 -1 -1>}
      localJointOffset=<4 0 0>
     }]
    relativeSpaceBone {boneName=spine_05}
    simSpaceSettings {simSpaceAngularAlpha=1}
   }
   n22 = Anim Dynamics - Bone: dyn_upper_shard_l(ComponentPose=@n21, Alpha=1, ExternalForce="0,0,0") {
    angularDampingOverride=1
    angularSpringConstant=20000
    bAngularSpring=true
    bOverrideAngularDamping=true
    bOverrideLinearDamping=true
    bUseGravityOverride=true
    gravityScale=0.35
    linearDampingOverride=1
    simulationSpace=RootRelative
    boundBone {boneName=dyn_upper_shard_l}
    componentLinearAccScale=<1 1 1>
    componentLinearVelScale=<1 1 1>
    physicsBodyDefinitions [{
      boundBone {boneName=dyn_upper_shard_l}
      boxExtents=<20 30 30>
      constraintSetup {angularLimitsMax=<5 5 5>, angularLimitsMin=<-5 -5 -5>, angularTarget=<1 0 0>, linearAxesMax=<1 1 1>, linearAxesMin=<-1 -1 -1>}
      localJointOffset=<2 0 0>
     }]
    relativeSpaceBone {boneName=spine_05}
    simSpaceSettings {simSpaceAngularAlpha=1}
   }
   n23 = Anim Dynamics - Bone: dyn_upper_shard_r(ComponentPose=@n22, Alpha=1, ExternalForce="0,0,0") {
    angularDampingOverride=1
    angularSpringConstant=20000
    bAngularSpring=true
    bOverrideAngularDamping=true
    bOverrideLinearDamping=true
    bUseGravityOverride=true
    gravityScale=0.35
    linearDampingOverride=1
    simulationSpace=RootRelative
    boundBone {boneName=dyn_upper_shard_r}
    componentLinearAccScale=<1 1 1>
    componentLinearVelScale=<1 1 1>
    physicsBodyDefinitions [{
      boundBone {boneName=dyn_upper_shard_r}
      boxExtents=<20 30 30>
      constraintSetup {angularLimitsMax=<5 5 5>, angularLimitsMin=<-5 -5 -5>, angularTarget=<1 0 0>, linearAxesMax=<1 1 1>, linearAxesMin=<-1 -1 -1>}
      localJointOffset=<2 0 0>
     }]
    relativeSpaceBone {boneName=spine_05}
    simSpaceSettings {simSpaceAngularAlpha=1}
   }
   n24 = Anim Dynamics - Bone: dyn_middle_shard_r(ComponentPose=@n23, Alpha=1, ExternalForce="0,0,0") {
    angularDampingOverride=1
    angularSpringConstant=20000
    bAngularSpring=true
    bOverrideAngularDamping=true
    bOverrideLinearDamping=true
    bUseGravityOverride=true
    gravityScale=0.35
    linearDampingOverride=1
    simulationSpace=RootRelative
    boundBone {boneName=dyn_middle_shard_r}
    componentLinearAccScale=<1 1 1>
    componentLinearVelScale=<1 1 1>
    physicsBodyDefinitions [{
      boundBone {boneName=dyn_middle_shard_r}
      boxExtents=<20 30 30>
      constraintSetup {angularLimitsMax=<5 5 5>, angularLimitsMin=<-5 -5 -5>, angularTarget=<1 0 0>, linearAxesMax=<1 1 1>, linearAxesMin=<-1 -1 -1>}
      localJointOffset=<4 0 0>
     }]
    relativeSpaceBone {boneName=spine_05}
    simSpaceSettings {simSpaceAngularAlpha=1}
   }
   n25 = Anim Dynamics - Bone: dyn_shard_arm_l_01(ComponentPose=@n24, Alpha=1, ExternalForce="0,0,0") {
    angularDampingOverride=1
    angularSpringConstant=20000
    bAngularSpring=true
    bOverrideAngularDamping=true
    bOverrideLinearDamping=true
    bUseGravityOverride=true
    gravityScale=0.35
    linearDampingOverride=1
    simulationSpace=RootRelative
    boundBone {boneName=dyn_shard_arm_l_01}
    componentLinearAccScale=<1 1 1>
    componentLinearVelScale=<1 1 1>
    physicsBodyDefinitions [{
      boundBone {boneName=dyn_shard_arm_l_01}
      boxExtents=<20 30 30>
      constraintSetup {angularLimitsMax=<5 5 5>, angularLimitsMin=<-5 -5 -5>, angularTarget=<1 0 0>, linearAxesMax=<1 1 1>, linearAxesMin=<-1 -1 -1>}
      localJointOffset=<2 0 0>
     }]
    relativeSpaceBone {boneName=spine_05}
    simSpaceSettings {simSpaceAngularAlpha=1}
   }
   n26 = Anim Dynamics - Bone: dyn_shard_arm_r_01(ComponentPose=@n25, Alpha=1, ExternalForce="0,0,0") {
    angularDampingOverride=1
    angularSpringConstant=20000
    bAngularSpring=true
    bOverrideAngularDamping=true
    bOverrideLinearDamping=true
    bUseGravityOverride=true
    gravityScale=0.35
    linearDampingOverride=1
    simulationSpace=RootRelative
    boundBone {boneName=dyn_shard_arm_r_01}
    componentLinearAccScale=<1 1 1>
    componentLinearVelScale=<1 1 1>
    physicsBodyDefinitions [{
      boundBone {boneName=dyn_shard_arm_r_01}
      boxExtents=<20 30 30>
      constraintSetup {angularLimitsMax=<5 5 5>, angularLimitsMin=<-5 -5 -5>, angularTarget=<1 0 0>, linearAxesMax=<1 1 1>, linearAxesMin=<-1 -1 -1>}
      localJointOffset=<4 0 0>
     }]
    relativeSpaceBone {boneName=spine_05}
    simSpaceSettings {simSpaceAngularAlpha=1}
   }
   n27 = Anim Dynamics - Bone: dyn_shard_bottom_m_01(ComponentPose=@n26, Alpha=1, ExternalForce="0,0,0") {
    angularDampingOverride=1
    angularSpringConstant=20000
    bAngularSpring=true
    bOverrideAngularDamping=true
    bOverrideLinearDamping=true
    bUseGravityOverride=true
    gravityScale=0.35
    linearDampingOverride=1
    simulationSpace=RootRelative
    boundBone {boneName=dyn_shard_bottom_m_01}
    componentLinearAccScale=<1 1 1>
    componentLinearVelScale=<1 1 1>
    physicsBodyDefinitions [{
      boundBone {boneName=dyn_shard_bottom_m_01}
      boxExtents=<20 30 30>
      constraintSetup {angularLimitsMax=<5 5 5>, angularLimitsMin=<-5 -5 -5>, angularTarget=<1 0 0>, linearAxesMax=<1 1 1>, linearAxesMin=<-1 -1 -1>}
      localJointOffset=<4 0 0>
     }]
    relativeSpaceBone {boneName=spine_05}
    simSpaceSettings {simSpaceAngularAlpha=1}
   }
   n28 = Anim Dynamics - Bone: dyn_shard_bottom_l_01(ComponentPose=@n27, Alpha=1, ExternalForce="0,0,0") {
    angularDampingOverride=1
    angularSpringConstant=20000
    bAngularSpring=true
    bOverrideAngularDamping=true
    bOverrideLinearDamping=true
    bUseGravityOverride=true
    gravityScale=0.35
    linearDampingOverride=1
    simulationSpace=RootRelative
    boundBone {boneName=dyn_shard_bottom_l_01}
    componentLinearAccScale=<1 1 1>
    componentLinearVelScale=<1 1 1>
    physicsBodyDefinitions [{
      boundBone {boneName=dyn_shard_bottom_l_01}
      boxExtents=<20 30 30>
      constraintSetup {angularLimitsMax=<5 5 5>, angularLimitsMin=<-5 -5 -5>, angularTarget=<1 0 0>, linearAxesMax=<1 1 1>, linearAxesMin=<-1 -1 -1>}
      localJointOffset=<4 0 0>
     }]
    relativeSpaceBone {boneName=spine_05}
    simSpaceSettings {simSpaceAngularAlpha=1}
   }
   n29 = Anim Dynamics - Bone: dyn_shard_bottom_r_01(ComponentPose=@n28, Alpha=1, ExternalForce="0,0,0") {
    angularDampingOverride=1
    angularSpringConstant=20000
    bAngularSpring=true
    bOverrideAngularDamping=true
    bOverrideLinearDamping=true
    bUseGravityOverride=true
    gravityScale=0.35
    linearDampingOverride=1
    simulationSpace=RootRelative
    boundBone {boneName=dyn_shard_bottom_r_01}
    componentLinearAccScale=<1 1 1>
    componentLinearVelScale=<1 1 1>
    physicsBodyDefinitions [{
      boundBone {boneName=dyn_shard_bottom_r_01}
      boxExtents=<20 30 30>
      constraintSetup {angularLimitsMax=<5 5 5>, angularLimitsMin=<-5 -5 -5>, angularTarget=<1 0 0>, linearAxesMax=<1 1 1>, linearAxesMin=<-1 -1 -1>}
      localJointOffset=<4 0 0>
     }]
    relativeSpaceBone {boneName=spine_05}
    simSpaceSettings {simSpaceAngularAlpha=1}
   }
   n30 = RigidBody!(ComponentPose=@n29, OverrideWorldGravity:Vector="0,0,-980", ExternalForce) -> (Pose) {
    alphaInputType=Bool
    bClampLinearTranslationLimitToRefPose=true
    bOverrideWorldGravity=true
    overridePhysicsAsset=&/M_MED_SheerWill_Blot_Physics
    simulationSpace=BaseBoneSpace
    alphaBoolBlend {blendOption=Linear}
    alphaScaleBiasClamp {bMapRange=true}
    baseBoneRef {boneName=pelvis}
    componentAppliedLinearAccClamp=<0 0 0>
    overrideWorldGravity {z=-980}
    simSpaceSettings {
     dampingAlpha=0
     maxAngularAcceleration=40
     maxAngularVelocity=8
     maxLinearAcceleration=1000
     maxLinearVelocity=700
     velocityScaleZ=0.5
     worldAlpha=1
     externalLinearDragV=<1 1 1>
    }
   }
   n31 = "Transform (Modify) Bone - Bone: base_loincloth_front"(ComponentPose=@n30, Alpha=1, Translation="0,0,0", Rotation="0.000,0.000,-10.000", Scale="1,1,1") {
    rotationMode=BMM_Additive
    rotationSpace=BCS_BoneSpace
    scaleMode=BMM_Replace
    alphaBoolBlend {blendOption=Linear}
    boneToModify {boneName=base_loincloth_front}
    rotation {roll=-10, yaw=0}
   }
   n32 = Get inGame()
   n33 = "Transform (Modify) Bone - Bone: head"(ComponentPose=@n31, Alpha=@n32, Translation="0,0,0", Rotation=Get Head Rot() -> (Head Rot:Rotator), Scale="1,1,1") {rotationMode=BMM_Additive, rotationSpace=BCS_BoneSpace, boneToModify {boneName=head}}
   n34 = Get Neck 2 Rot() -> (Neck 2 Rot:Rotator)
   n35 = "Transform (Modify) Bone - Bone: dyn_shard_l_1"(
    ComponentPose="Transform (Modify) Bone - Bone: neck_01"(
     ComponentPose="Transform (Modify) Bone - Bone: neck_02"(ComponentPose=@n33, Alpha=@n32, Translation="0,0,0", Rotation=@n34, Scale="1,1,1") {rotationMode=BMM_Additive, rotationSpace=BCS_BoneSpace, boneToModify {boneName=neck_02}}
     Alpha=@n32
     Translation="0,0,0"
     Rotation=Get Neck 1 Rot() -> (Neck 1 Rot:Rotator)
     Scale="1,1,1"
    ) {rotationMode=BMM_Additive, rotationSpace=BCS_BoneSpace, boneToModify {boneName=neck_01}}
    Alpha=1
    Translation=Get ShardFloatA() -> (ShardFloatA:Vector)
    Rotation="0,0,0"
    Scale="1,1,1"
   ) {
    rotationMode=BMM_Additive
    scaleSpace=BCS_BoneSpace
    translationMode=BMM_Additive
    translationSpace=BCS_BoneSpace
    alphaBoolBlend {blendOption=Linear}
    boneToModify {boneName=dyn_shard_l_1}
   }
   n36 = "Transform (Modify) Bone - Bone: dyn_shard_r_4"(ComponentPose=@n35, Alpha=1, Translation=Get ShardFloatA(), Rotation="0,0,0", Scale="1,1,1") {
    rotationMode=BMM_Additive
    scaleSpace=BCS_BoneSpace
    translationMode=BMM_Additive
    translationSpace=BCS_BoneSpace
    alphaBoolBlend {blendOption=Linear}
    boneToModify {boneName=dyn_shard_r_4}
   }
   n37 = "Transform (Modify) Bone - Bone: dyn_shard_l_2"(ComponentPose=@n36, Alpha=1, Translation=Get ShardFloatC() -> (ShardFloatC:Vector), Rotation="0,0,0", Scale="1,1,1") {
    rotationMode=BMM_Additive
    scaleSpace=BCS_BoneSpace
    translationMode=BMM_Additive
    translationSpace=BCS_BoneSpace
    alphaBoolBlend {blendOption=Linear}
    boneToModify {boneName=dyn_shard_l_2}
   }
   n38 = "Transform (Modify) Bone - Bone: dyn_shard_l_3"(ComponentPose=@n37, Alpha=1, Translation=Get ShardFloatB() -> (ShardFloatB:Vector), Rotation="0,0,0", Scale="1,1,1") {
    rotationMode=BMM_Additive
    scaleSpace=BCS_BoneSpace
    translationMode=BMM_Additive
    translationSpace=BCS_BoneSpace
    alphaBoolBlend {blendOption=Linear}
    boneToModify {boneName=dyn_shard_l_3}
   }
   n39 = "Transform (Modify) Bone - Bone: dyn_shard_r_3"(ComponentPose=@n38, Alpha=1, Translation=Get ShardFloatC(), Rotation="0,0,0", Scale="1,1,1") {
    rotationMode=BMM_Additive
    scaleSpace=BCS_BoneSpace
    translationMode=BMM_Additive
    translationSpace=BCS_BoneSpace
    alphaBoolBlend {blendOption=Linear}
    boneToModify {boneName=dyn_shard_r_3}
   }
   n40 = "Transform (Modify) Bone - Bone: dyn_shard_arm_l_01"(ComponentPose=@n39, Alpha=1, Translation=Get ShardFloatA(), Rotation="0,0,0", Scale="1,1,1") {
    rotationMode=BMM_Additive
    scaleSpace=BCS_BoneSpace
    translationMode=BMM_Additive
    translationSpace=BCS_BoneSpace
    alphaBoolBlend {blendOption=Linear}
    boneToModify {boneName=dyn_shard_arm_l_01}
   }
   n41 = "Transform (Modify) Bone - Bone: dyn_shard_arm_r_01"(ComponentPose=@n40, Alpha=1, Translation=Get ShardFloatA(), Rotation="0,0,0", Scale="1,1,1") {
    rotationMode=BMM_Additive
    scaleSpace=BCS_BoneSpace
    translationMode=BMM_Additive
    translationSpace=BCS_BoneSpace
    alphaBoolBlend {blendOption=Linear}
    boneToModify {boneName=dyn_shard_arm_r_01}
   }
   n42 = "Transform (Modify) Bone - Bone: dyn_upper_shard_r"(ComponentPose=@n41, Alpha=1, Translation=Get ShardFloatB(), Rotation="0,0,0", Scale="1,1,1") {
    rotationMode=BMM_Additive
    scaleSpace=BCS_BoneSpace
    translationMode=BMM_Additive
    translationSpace=BCS_BoneSpace
    alphaBoolBlend {blendOption=Linear}
    boneToModify {boneName=dyn_upper_shard_r}
   }
   n43 = "Transform (Modify) Bone - Bone: dyn_shard_bottom_l_01"(ComponentPose=@n42, Alpha=1, Translation=Get ShardFloatA(), Rotation="0,0,0", Scale="1,1,1") {
    rotationMode=BMM_Additive
    scaleSpace=BCS_BoneSpace
    translationMode=BMM_Additive
    translationSpace=BCS_BoneSpace
    alphaBoolBlend {blendOption=Linear}
    boneToModify {boneName=dyn_shard_bottom_l_01}
   }
   n44 = "Transform (Modify) Bone - Bone: dyn_middle_shard_l"(ComponentPose=@n43, Alpha=1, Translation=Get ShardFloatC(), Rotation="0,0,0", Scale="1,1,1") {
    rotationMode=BMM_Additive
    scaleSpace=BCS_BoneSpace
    translationMode=BMM_Additive
    translationSpace=BCS_BoneSpace
    alphaBoolBlend {blendOption=Linear}
    boneToModify {boneName=dyn_middle_shard_l}
   }
   n45 = "Transform (Modify) Bone - Bone: dyn_shard_bottom_r_01"(ComponentPose=@n44, Alpha=1, Translation=Get ShardFloatA(), Rotation="0,0,0", Scale="1,1,1") {
    rotationMode=BMM_Additive
    scaleSpace=BCS_BoneSpace
    translationMode=BMM_Additive
    translationSpace=BCS_BoneSpace
    alphaBoolBlend {blendOption=Linear}
    boneToModify {boneName=dyn_shard_bottom_r_01}
   }
   n46 = "Transform (Modify) Bone - Bone: dyn_upper_shard_l"(ComponentPose=@n45, Alpha=1, Translation=Get ShardFloatA(), Rotation="0,0,0", Scale="1,1,1") {
    rotationMode=BMM_Additive
    scaleSpace=BCS_BoneSpace
    translationMode=BMM_Additive
    translationSpace=BCS_BoneSpace
    alphaBoolBlend {blendOption=Linear}
    boneToModify {boneName=dyn_upper_shard_l}
   }
   n47 = "Transform (Modify) Bone - Bone: dyn_middle_shard_r"(ComponentPose=@n46, Alpha=1, Translation=Get ShardFloatB(), Rotation="0,0,0", Scale="1,1,1") {
    rotationMode=BMM_Additive
    scaleSpace=BCS_BoneSpace
    translationMode=BMM_Additive
    translationSpace=BCS_BoneSpace
    alphaBoolBlend {blendOption=Linear}
    boneToModify {boneName=dyn_middle_shard_r}
   }
   n48 = "Transform (Modify) Bone - Bone: dyn_shard_bottom_m_01"(ComponentPose=@n47, Alpha=1, Translation=Get ShardFloatC(), Rotation="0,0,0", Scale="1,1,1") {
    rotationMode=BMM_Additive
    scaleSpace=BCS_BoneSpace
    translationMode=BMM_Additive
    translationSpace=BCS_BoneSpace
    alphaBoolBlend {blendOption=Linear}
    boneToModify {boneName=dyn_shard_bottom_m_01}
   }
   n49 = Get Head_Mesh() -> (Head_Mesh:Skeletal Mesh Component Object Reference)
   Save cached pose 'CACHE_NormalSetup'(Pose=Layered blend per bone(BasePose=Component To Local(ComponentPose=@n48), BlendPoses_0=Copy Pose From Mesh(SourceMeshComponent=@n49)) {layerSetup [{branchFilters [{boneName=FACIAL_C_FacialRoot}]}]}) {cacheName=CACHE_NormalSetup}
   Save cached pose 'Base Pose'(Pose=Copy Pose From Mesh() {bUseAttachedParent=true}) {cacheName=Base Pose}
   Get Head Rot()
  ]}
 Dyn Shards {nodes [
   Dyn Shards() {is=title K2Node_FunctionEntry}
   > Set PawnSpeedRemap(PawnSpeedRemap:double=MapRangeClamped(Value=Get PawnSpeed2D(), InRangeA=0, InRangeB=549)) -> (Output_Get:double)
   n1 = GetGameTimeInSeconds()
   n2 = "Sin (Radians)"(A=float * float(A=float * float(A=0.5, B=1), B=@n1))
   n3 = "Sin (Radians)"(A=float * float(A=float * float(A=0.5, B=2), B=@n1))
   n4 = "Sin (Radians)"(A=float * float(A=float * float(A=0.5, B=3), B=@n1))
   > Set ShardFloatA(ShardFloatA:Vector=MakeVector(X=float * float(A=@n4, B=0.3), Y=float * float(A=@n2, B=0.56), Z=float * float(A=@n3, B=0.22))) -> (Output_Get:Vector)
   n5 = "Cos (Radians)"!(A:double=float * float(A=float * float(A=0.5, B=1.5), B=@n1)) -> (ReturnValue:double)
   n6 = "Cos (Radians)"(A=float * float(A=float * float(A=0.5, B=1.78), B=@n1))
   n7 = "Cos (Radians)"(A=float * float(A=float * float(A=0.5, B=1.15), B=@n1))
   > Set ShardFloatB(ShardFloatB:Vector=MakeVector(X=float * float(A=@n5, B=0.6), Y=float * float(A=@n6, B=0.725), Z=float * float(A=@n7, B=0.3))) -> (Output_Get:Vector)
   n8 = "Sin (Radians)"(A=float * float(A=float * float(A=0.5, B=1.5), B=@n1))
   n9 = "Sin (Radians)"(A=float * float(A=float * float(A=0.5, B=2), B=@n1))
   n10 = "Sin (Radians)"(A=float * float(A=float * float(A=0.5, B=1), B=@n1))
   > Set ShardFloatC(ShardFloatC:Vector=MakeVector(X=float * float(A=@n8, B=0.3), Y=float * float(A=@n10, B=0.6), Z=float * float(A=@n9, B=0.25))) -> (Output_Get:Vector)
  ]}
 Dynamic Shards {nodes [
   n1 = Get ShardFloatA()
   n2 = "Transform (Modify) Bone - Bone: dyn_shard_l_1"(ComponentPose=Local To Component(LocalPose=Input Pose()), Alpha=1, Translation=@n1, Rotation="0,0,0", Scale="1,1,1") {
    rotationMode=BMM_Additive
    scaleSpace=BCS_BoneSpace
    translationMode=BMM_Additive
    translationSpace=BCS_BoneSpace
    alphaBoolBlend {blendOption=Linear}
    boneToModify {boneName=dyn_shard_l_1}
   }
   n3 = Get ShardFloatC()
   n4 = "Transform (Modify) Bone - Bone: dyn_shard_l_2"(ComponentPose=@n2, Alpha=1, Translation=@n3, Rotation="0,0,0", Scale="1,1,1") {
    rotationMode=BMM_Additive
    scaleSpace=BCS_BoneSpace
    translationMode=BMM_Additive
    translationSpace=BCS_BoneSpace
    alphaBoolBlend {blendOption=Linear}
    boneToModify {boneName=dyn_shard_l_2}
   }
   n5 = "Transform (Modify) Bone - Bone: dyn_shard_l_3"(ComponentPose=@n4, Alpha=1, Translation=Get ShardFloatB(), Rotation="0,0,0", Scale="1,1,1") {
    rotationMode=BMM_Additive
    scaleSpace=BCS_BoneSpace
    translationMode=BMM_Additive
    translationSpace=BCS_BoneSpace
    alphaBoolBlend {blendOption=Linear}
    boneToModify {boneName=dyn_shard_l_3}
   }
   n6 = "Transform (Modify) Bone - Bone: dyn_shard_r_3"(ComponentPose=@n5, Alpha=1, Translation=@n3, Rotation="0,0,0", Scale="1,1,1") {
    rotationMode=BMM_Additive
    scaleSpace=BCS_BoneSpace
    translationMode=BMM_Additive
    translationSpace=BCS_BoneSpace
    alphaBoolBlend {blendOption=Linear}
    boneToModify {boneName=dyn_shard_r_3}
   }
   n7 = "Transform (Modify) Bone - Bone: dyn_shard_r_4"(ComponentPose=@n6, Alpha=1, Translation=@n1, Rotation="0,0,0", Scale="1,1,1") {
    rotationMode=BMM_Additive
    scaleSpace=BCS_BoneSpace
    translationMode=BMM_Additive
    translationSpace=BCS_BoneSpace
    alphaBoolBlend {blendOption=Linear}
    boneToModify {boneName=dyn_shard_r_4}
   }
   n8 = Get ShardFloatA()
   n9 = "Transform (Modify) Bone - Bone: dyn_shard_arm_l_01"(ComponentPose=@n7, Alpha=1, Translation=@n8, Rotation="0,0,0", Scale="1,1,1") {
    rotationMode=BMM_Additive
    scaleSpace=BCS_BoneSpace
    translationMode=BMM_Additive
    translationSpace=BCS_BoneSpace
    alphaBoolBlend {blendOption=Linear}
    boneToModify {boneName=dyn_shard_arm_l_01}
   }
   n10 = "Transform (Modify) Bone - Bone: dyn_shard_arm_r_01"(ComponentPose=@n9, Alpha=1, Translation=@n8, Rotation="0,0,0", Scale="1,1,1") {
    rotationMode=BMM_Additive
    scaleSpace=BCS_BoneSpace
    translationMode=BMM_Additive
    translationSpace=BCS_BoneSpace
    alphaBoolBlend {blendOption=Linear}
    boneToModify {boneName=dyn_shard_arm_r_01}
   }
   n11 = Get ShardFloatB()
   n12 = "Transform (Modify) Bone - Bone: dyn_upper_shard_r"(ComponentPose=@n10, Alpha=1, Translation=@n11, Rotation="0,0,0", Scale="1,1,1") {
    rotationMode=BMM_Additive
    scaleSpace=BCS_BoneSpace
    translationMode=BMM_Additive
    translationSpace=BCS_BoneSpace
    alphaBoolBlend {blendOption=Linear}
    boneToModify {boneName=dyn_upper_shard_r}
   }
   n13 = Get ShardFloatA()
   n14 = "Transform (Modify) Bone - Bone: dyn_shard_bottom_l_01"(ComponentPose=@n12, Alpha=1, Translation=@n13, Rotation="0,0,0", Scale="1,1,1") {
    rotationMode=BMM_Additive
    scaleSpace=BCS_BoneSpace
    translationMode=BMM_Additive
    translationSpace=BCS_BoneSpace
    alphaBoolBlend {blendOption=Linear}
    boneToModify {boneName=dyn_shard_bottom_l_01}
   }
   n15 = Get ShardFloatC()
   n16 = "Transform (Modify) Bone - Bone: dyn_middle_shard_l"(ComponentPose=@n14, Alpha=1, Translation=@n15, Rotation="0,0,0", Scale="1,1,1") {
    rotationMode=BMM_Additive
    scaleSpace=BCS_BoneSpace
    translationMode=BMM_Additive
    translationSpace=BCS_BoneSpace
    alphaBoolBlend {blendOption=Linear}
    boneToModify {boneName=dyn_middle_shard_l}
   }
   n17 = "Transform (Modify) Bone - Bone: dyn_shard_bottom_r_01"(ComponentPose=@n16, Alpha=1, Translation=@n13, Rotation="0,0,0", Scale="1,1,1") {
    rotationMode=BMM_Additive
    scaleSpace=BCS_BoneSpace
    translationMode=BMM_Additive
    translationSpace=BCS_BoneSpace
    alphaBoolBlend {blendOption=Linear}
    boneToModify {boneName=dyn_shard_bottom_r_01}
   }
   n18 = "Transform (Modify) Bone - Bone: dyn_upper_shard_l"(ComponentPose=@n17, Alpha=1, Translation=@n13, Rotation="0,0,0", Scale="1,1,1") {
    rotationMode=BMM_Additive
    scaleSpace=BCS_BoneSpace
    translationMode=BMM_Additive
    translationSpace=BCS_BoneSpace
    alphaBoolBlend {blendOption=Linear}
    boneToModify {boneName=dyn_upper_shard_l}
   }
   n19 = "Transform (Modify) Bone - Bone: dyn_middle_shard_r"(ComponentPose=@n18, Alpha=1, Translation=@n11, Rotation="0,0,0", Scale="1,1,1") {
    rotationMode=BMM_Additive
    scaleSpace=BCS_BoneSpace
    translationMode=BMM_Additive
    translationSpace=BCS_BoneSpace
    alphaBoolBlend {blendOption=Linear}
    boneToModify {boneName=dyn_middle_shard_r}
   }
   n20 = "Transform (Modify) Bone - Bone: dyn_shard_bottom_m_01"(ComponentPose=@n19, Alpha=1, Translation=@n15, Rotation="0,0,0", Scale="1,1,1") {
    rotationMode=BMM_Additive
    scaleSpace=BCS_BoneSpace
    translationMode=BMM_Additive
    translationSpace=BCS_BoneSpace
    alphaBoolBlend {blendOption=Linear}
    boneToModify {boneName=dyn_shard_bottom_m_01}
   }
   Output Pose(Result=Component To Local(ComponentPose=@n20))
  ]}
 EventGraph {nodes [
   Event BlueprintUpdateAnimation()
   > n1 = Sequence!(execute) -> (then_0, then_1, then_2:Exec, then_3:Exec, then_4:Exec, then_5:Exec)
   > "Gravity Override State Function (Parameter Asset) (Nativized)"(
    CharacterPartAnimInstance=Self-Reference()
    GameTimeInSeconds=GetGameTimeInSeconds()
    Parameters=/BRCosmetics/Characters/Player/Male/Medium/Bodies/M_MED_ScotchPeek/Meshes/F_MED_ScotchPeek_Parameters
   )
   > Set Gravity Override(Gravity Override="0,0,-980")
   n2 = Neck Values!(execute=@n1.then_3, self, Bone Name:Name=neck_02, "Base Name ":Name=neck_01) -> (then, Joint Rotation:Rotator)
   > Set Neck 2 Rot(Neck 2 Rot:Rotator=@n2) -> (Output_Get:Rotator)
   > n3 = Neck Values(Bone Name=neck_01, "Base Name "=neck_01)
   > Set Neck 1 Rot(Neck 1 Rot:Rotator=@n3) -> (Output_Get:Rotator)
   Dyn Shards(execute=@n1.then_4) {is=title K2Node_CallFunction}
   n4 = GetOwnerSkeletalMeshComponent()
   n5 = BreakRotator(InRot=BreakTransform(InTransform=GetDeltaTransformFromRefPose(self=@n4, BoneName=head, BaseName=neck_02)).Rotation)
   n6 = Branch(execute=@n1.then_2, Condition="float < float"(A=@n5.Yaw))
   > n7 = Set z Percentage(z Percentage:double=0.2) -> (Output_Get:double)
   n8 = Get y Percentage() -> (y Percentage:double)
   n9 = Get z Percentage() -> (z Percentage:double)
   n10 = Get x Percentage() -> (x Percentage:double)
   Set Head Rot(
    execute=[@n13.then, @n7.then]
    Head Rot:Rotator=MakeRotator(
     Roll=float * float!(A=@n5, B=-1, C:double=@n10) -> (ReturnValue) {is=title K2Node_PromotableOperator}
     Pitch=float * float(A=@n5.Pitch, B=-1, C=@n8)
     Yaw=float * float(A=@n5.Yaw, B=-1, C=@n9)
    )
   ) -> (Output_Get:Rotator)
   n11 = NOT Boolean(A=Get bIsPlayingEmote())
   n12 = Get bIsInFrontEnd()
   > Set inGame(inGame:double="To Float (Boolean)"(InBool=OR Boolean(A=@n11, B=@n12))) -> (Output_Get:double)
   n13 = Set z Percentage(execute=@n6.else, z Percentage=0)
   Branch(execute=@n1.then_5, Condition=Get bIsInFrontEnd())
   > Set Custom Idle Is Playing Float(Custom Idle Is Playing Float:double=GetCurveValue(self=GetOwnerFrontEndAnimInstance(), CurveName=FrontEnd_Idle_Bypass)) -> (Output_Get:double)
   Event OnCharacterCustomizationLoaded()
   > n14 = Cast To PlayerPawn_Athena(Object=TryGetPawnOwner())
   > Set Head_Mesh(Head_Mesh:Skeletal Mesh Component Object Reference=GetSkeletalMeshForPartType(self=@n14, PartType=Head)) -> (Output_Get:Skeletal Mesh Component Object Reference)
  ]}
 Neck Values {nodes [
   n1 = Neck Values!() -> (then, Bone Name:Name, "Base Name ":Name) {is=title K2Node_FunctionEntry}
   n2 = GetOwnerSkeletalMeshComponent()
   n3 = BreakRotator(InRot=BreakTransform(InTransform=GetDeltaTransformFromRefPose(self=@n2, BoneName=@n1, BaseName=@n1."Base Name ")).Rotation)
   > Branch(Condition="float < float"(A=@n3.Yaw))
   > n4 = Set z Percentage_neck(z Percentage_neck:double=1.2) -> (Output_Get:double)
   n5 = Get z Percentage_neck() -> (z Percentage_neck:double)
   n6 = Get x Percentage_neck() -> (x Percentage_neck:double)
   n7 = Get y Percentage_neck() -> (y Percentage_neck:double)
   Return Node!(
    execute=[@n8.then, @n4.then]
    Joint Rotation:Rotator=MakeRotator(Roll=float * float(A=@n3, B=-1, C=@n6), Pitch=float * float(A=@n3.Pitch, B=-1, C=@n7), Yaw=float * float(A=@n3.Yaw, B=-1, C=@n5))
   )
   n8 = Set z Percentage_neck(z Percentage_neck=0)
  ]}
]
variables [
 alpha {type=number}
 Custom Idle Is Playing Float {type=number}
 Gravity Override {type=Vector}
 Head Rot {type=Rotator}
 Head_Mesh {type=SkeletalMeshComponent}
 inGame {type=number}
 Neck 1 Rot {type=Rotator}
 Neck 2 Rot {type=Rotator}
 PawnSpeedRemap {type=number}
 Range {type=number}
 ShardFloatA {type=Vector}
 ShardFloatB {type=Vector}
 ShardFloatC {type=Vector}
 "Translate Shard 1 " {type=Vector}
 Translate Shard 2 {type=Vector}
 Translate Shard 3 {type=Vector}
 Translate Shard 4 {type=Vector}
 Translate Shard 5 {type=Vector}
 Translate Shard 6 {type=Vector}
 x Percentage {type=number, default=0.8}
 x Percentage_neck {type=number, default=0.3}
 y Percentage {type=number, default=0.8}
 y Percentage_neck {type=number, default=0.3}
 z Percentage {type=number, default=0.5}
 z Percentage_neck {type=number}
]
