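# Physics asset for the face accessory part of the male medium SheerWill Blot player cosmetic, providing bodies for dynamic hair.
# Capsules form three hair chains (two-link left and right, tapering five-link middle) off base bones constrained to the head with free angular motion.
# Chain joints have narrow swing limits and almost no twist.
# Capsules on head, neck, upper spine, clavicles and left upper arm cover the surrounding body, mostly unconstrained.
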PhysicsAsset /BRCosmetics/Characters/Player/Male/Medium/Bodies/M_MED_SheerWill_Blot/Meshes/Parts/M_MED_SheerWill_Blot_FaceAcc_Physics
bodies {
 base_hair_l [{shapeType=Capsule, length=0.505, radius=0.505, shapeName=base_hair_l_capsule, rotation=<rot 0 -90 0>}]
 base_hair_m [{shapeType=Capsule, length=0.505, radius=0.505, shapeName=base_hair_m_capsule, rotation=<rot 0 -90 0>}]
 base_hair_r [{shapeType=Capsule, length=0.505, radius=0.505, shapeName=base_hair_r_capsule, rotation=<rot 0 -90 0>}]
 clavicle_l [{shapeType=Capsule, length=8.8922, radius=8.8922, shapeName=clavicle_l_capsule, center=<14.3312 0 1.6312>, rotation=<rot -80 0 0>}]
 clavicle_r [{shapeType=Capsule, length=0.505, radius=0.505, shapeName=clavicle_r_capsule, rotation=<rot 0 -90 0>}]
 dyn_hair_l_01 [{
   shapeType=Capsule
   length=3.3325
   radius=2.0512
   shapeName=dyn_hair_l_01_capsule
   center=<2.031 0.495 1.1138>
   rotation=<rot 0 -79.4139 -77.6672>
  }]
 dyn_hair_l_02 [{shapeType=Capsule, length=3.4051, radius=2, shapeName=dyn_hair_l_02_capsule, center=<0.8137 0 0>, rotation=<rot 0 -64.1462 -69.9353>}]
 dyn_hair_m_01 [{
   shapeType=Capsule
   length=3.0602
   radius=1.9093
   shapeName=dyn_hair_m_01_capsule
   center=<1.3071 0.263 -0.4475>
   rotation=<rot 0 72.092 -57.4124>
  }]
 dyn_hair_m_02 [{
   shapeType=Capsule
   length=3.6543
   radius=0.9218
   shapeName=dyn_hair_m_02_capsule
   center=<2.7702 0.1846 0.0653>
   rotation=<rot 0 91.4867 -89.2284>
  }]
 dyn_hair_m_03 [{
   shapeType=Capsule
   length=3.0438
   radius=0.7655
   shapeName=dyn_hair_m_03_capsule
   center=<2.3292 0.3417 0.2028>
   rotation=<rot 0 91.7207 -88.7669>
  }]
 dyn_hair_m_04 [{
   shapeType=Capsule
   length=3.314
   radius=0.6316
   shapeName=dyn_hair_m_04_capsule
   center=<1.729 0.5565 0.0269>
   rotation=<rot 0 92.0593 -87.7018>
  }]
 dyn_hair_m_05 [{
   shapeType=Capsule
   length=1.6758
   radius=0.505
   shapeName=dyn_hair_m_05_capsule
   center=<0.4929 0.7877 -0.1604>
   rotation=<rot 0 -85.9912 -75.192>
  }]
 dyn_hair_r_01 [{shapeType=Capsule, length=2.2044, radius=2.0315, shapeName=dyn_hair_r_01_capsule, center=<0 0 -0.7963>, rotation=<rot 90 0 0>}]
 dyn_hair_r_02 [{shapeType=Capsule, length=3.9449, radius=1.6328, shapeName=dyn_hair_r_02_capsule, center=<-0.4323 0 0>, rotation=<rot 0 -76.8557 -82.7014>}]
 head [{shapeType=Capsule, length=8.4323, radius=9, center=<3.6974 1.4464 0.5186>, rotation=<rot 90 6.4508 0>}]
 neck_02 [{shapeType=Capsule, length=3.264, radius=5.3374, shapeName=neck_02_capsule, rotation=<rot 90 0 0>}]
 spine_04 [{shapeType=Capsule, length=11.7257, radius=11.7257, shapeName=spine_04_capsule, center=<0 2.8387 0>}]
 spine_05 [{shapeType=Capsule, length=14.3187, radius=14.3187, shapeName=spine_05_capsule, center=<-2.1117 0.0046 0>}]
 upperarm_l [{shapeType=Capsule, length=8.2388, radius=8.2388, shapeName=upperarm_l_capsule, center=<4.1346 0 0>, rotation=<rot -90 0 0>}]
}
constraints [
 {
  bone1Name=base_hair_m
  bone2Name=head
  swing1LimitDegrees=45
  swing1Motion=Free
  swing2LimitDegrees=45
  swing2Motion=Free
  twistLimitDegrees=45
  twistMotion=Free
 }
 {bone1Name=dyn_hair_m_01, bone2Name=base_hair_m, swing1LimitDegrees=10, swing2LimitDegrees=30, twistLimitDegrees=1}
 {bone1Name=dyn_hair_m_02, bone2Name=dyn_hair_m_01, swing1LimitDegrees=10, swing2LimitDegrees=30, twistLimitDegrees=1}
 {bone1Name=dyn_hair_m_03, bone2Name=dyn_hair_m_02, swing1LimitDegrees=10, swing2LimitDegrees=30, twistLimitDegrees=1}
 {bone1Name=dyn_hair_m_04, bone2Name=dyn_hair_m_03, swing1LimitDegrees=10, swing2LimitDegrees=30, twistLimitDegrees=1}
 {bone1Name=dyn_hair_m_05, bone2Name=dyn_hair_m_04, swing1LimitDegrees=10, swing2LimitDegrees=30, twistLimitDegrees=1}
 {
  bone1Name=base_hair_l
  bone2Name=head
  swing1LimitDegrees=45
  swing1Motion=Free
  swing2LimitDegrees=45
  swing2Motion=Free
  twistLimitDegrees=45
  twistMotion=Free
 }
 {bone1Name=dyn_hair_l_01, bone2Name=base_hair_l, swing1LimitDegrees=25, swing2LimitDegrees=15, twistLimitDegrees=1}
 {bone1Name=dyn_hair_l_02, bone2Name=dyn_hair_l_01, swing1LimitDegrees=25, swing2LimitDegrees=15, twistLimitDegrees=1}
 {
  bone1Name=base_hair_r
  bone2Name=head
  swing1LimitDegrees=45
  swing1Motion=Free
  swing2LimitDegrees=45
  swing2Motion=Free
  twistLimitDegrees=45
  twistMotion=Free
 }
 {bone1Name=dyn_hair_r_01, bone2Name=base_hair_r, swing1LimitDegrees=25, swing2LimitDegrees=15, twistLimitDegrees=1}
 {bone1Name=dyn_hair_r_02, bone2Name=dyn_hair_r_01, swing1LimitDegrees=25, swing2LimitDegrees=15, twistLimitDegrees=1}
 {
  bone1Name=head
  bone2Name=neck_02
  swing1LimitDegrees=45
  swing1Motion=Free
  swing2LimitDegrees=45
  swing2Motion=Free
  twistLimitDegrees=45
  twistMotion=Free
 }
 {
  bone1Name=clavicle_r
  bone2Name=spine_05
  swing1LimitDegrees=45
  swing1Motion=Free
  swing2LimitDegrees=45
  swing2Motion=Free
  twistLimitDegrees=45
  twistMotion=Free
 }
]
