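# Animation Blueprint for the SheerWill Blot face accessory, adding secondary motion to its head shards and back hair.
# Copies the attached parent's pose through a Control Rig, then runs spring AnimDynamics on three shards and head-space RigidBody physics.
# Time-based sine offsets from Dyn Shards then make shards float, before weighted constraints blend back-hair bones between head, neck and spine.
# ShardFloatB, PawnSpeedRemap and the base_hair_r_back transform don't affect the output.
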AnimBlueprint /BRCosmetics/Characters/Player/Male/Medium/Bodies/M_MED_SheerWill_Blot/Meshes/Parts/M_MED_SheerWill_Blot_FaceAcc_AnimBP
target_skeleton=&/M_MED_SheerWill_Blot_FaceAcc_Skeleton
graphs [
 AnimGraph {nodes [
   "Transform (Modify) Bone - Bone: base_hair_r_back"(Alpha=1, Translation="0,0,0", Rotation="-0.000,-35.000,-0.000", Scale="1,1,1") {
    rotationMode=BMM_Replace
    rotationSpace=BCS_BoneSpace
    scaleMode=BMM_Additive
    scaleSpace=BCS_BoneSpace
    boneToModify {boneName=base_hair_r_back}
    rotation {yaw=-35.0001}
   }
   n1 = Control Rig(Source=Copy Pose From Mesh() {bUseAttachedParent=true}) {controlRigAssetReference {blueprintRigClass=&/F_MED_SheerWillBlot_FaceAcc_CtrlRig.F_MED_SheerWillBlot_FaceAcc_CtrlRig_C}}
   n2 = Anim Dynamics - Bone: dyn_shard_m(
    ComponentPose="Transform (Modify) Bone - Bone: base_shard_m"(ComponentPose=Local To Component(LocalPose=@n1), Alpha=1, Translation="0,1,0", Rotation="0,0,0", Scale="1,1,1") {translationMode=BMM_Additive, boneToModify {boneName=base_shard_m}, translation {y=1}}
    Alpha=1
    ExternalForce="0,0,0"
   ) {
    angularDampingOverride=1
    angularSpringConstant=20000
    bAngularSpring=true
    bOverrideAngularDamping=true
    bOverrideLinearDamping=true
    bUseGravityOverride=true
    gravityScale=0.35
    linearDampingOverride=1
    simulationSpace=RootRelative
    boundBone {boneName=dyn_shard_m}
    componentLinearAccScale=<1 1 1>
    componentLinearVelScale=<1 1 1>
    physicsBodyDefinitions [{
      boundBone {boneName=dyn_shard_m}
      boxExtents=<20 30 30>
      constraintSetup {angularLimitsMax=<0 5 5>, angularLimitsMin=<0 -5 -5>, angularTarget=<1 0 0>, linearAxesMax=<0 1 1>, linearAxesMin=<0 -1 -1>}
      localJointOffset=<2 0 0>
     }]
    relativeSpaceBone {boneName=head}
    simSpaceSettings {simSpaceAngularAlpha=1}
   }
   n3 = Anim Dynamics - Bone: dyn_shard_l(ComponentPose=@n2, Alpha=1, ExternalForce="0,0,0") {
    angularDampingOverride=1
    angularSpringConstant=20000
    bAngularSpring=true
    bOverrideAngularDamping=true
    bOverrideLinearDamping=true
    bUseGravityOverride=true
    gravityScale=0.35
    linearDampingOverride=1
    simulationSpace=RootRelative
    boundBone {boneName=dyn_shard_l}
    componentLinearAccScale=<1 1 1>
    componentLinearVelScale=<1 1 1>
    physicsBodyDefinitions [{
      boundBone {boneName=dyn_shard_l}
      boxExtents=<20 30 30>
      constraintSetup {angularLimitsMax=<5 5 5>, angularLimitsMin=<-5 -5 -5>, angularTarget=<1 0 0>, linearAxesMax=<1 2 1>, linearAxesMin=<-1 -1 -1>}
      localJointOffset=<1 0 0>
     }]
    relativeSpaceBone {boneName=head}
    simSpaceSettings {simSpaceAngularAlpha=1}
   }
   n4 = Anim Dynamics - Bone: dyn_shard_r(ComponentPose=@n3, Alpha=1, ExternalForce="0,0,0") {
    angularDampingOverride=1
    angularSpringConstant=20000
    bAngularSpring=true
    bOverrideAngularDamping=true
    bOverrideLinearDamping=true
    bUseGravityOverride=true
    gravityScale=0.35
    linearDampingOverride=1
    simulationSpace=RootRelative
    boundBone {boneName=dyn_shard_r}
    componentLinearAccScale=<1 1 1>
    componentLinearVelScale=<1 1 1>
    physicsBodyDefinitions [{
      boundBone {boneName=dyn_shard_r}
      boxExtents=<20 30 30>
      constraintSetup {angularLimitsMax=<5 5 5>, angularLimitsMin=<-5 -5 -5>, angularTarget=<1 0 0>, linearAxesMax=<1 1 1>, linearAxesMin=<-1 -1 -1>}
      localJointOffset=<4 0 0>
     }]
    relativeSpaceBone {boneName=head}
    simSpaceSettings {simSpaceAngularAlpha=1}
   }
   n5 = RigidBody!(ComponentPose=@n4, OverrideWorldGravity:Vector="0,0,-980", ExternalForce) -> (Pose) {
    alphaInputType=Bool
    bClampLinearTranslationLimitToRefPose=true
    bOverrideWorldGravity=true
    overridePhysicsAsset=&/M_MED_SheerWill_Blot_FaceAcc_Physics
    simulationSpace=BaseBoneSpace
    alphaBoolBlend {blendOption=Linear}
    alphaScaleBiasClamp {bMapRange=true}
    baseBoneRef {boneName=head}
    componentAppliedLinearAccClamp=<0 0 0>
    overrideWorldGravity {z=-980}
    simSpaceSettings {
     dampingAlpha=0
     maxAngularAcceleration=40
     maxAngularVelocity=8
     maxLinearAcceleration=3500
     maxLinearVelocity=700
     velocityScaleZ=0.5
     worldAlpha=1
     externalLinearDragV=<1 1 1>
    }
   }
   n6 = "Transform (Modify) Bone - Bone: dyn_shard_r"(ComponentPose=@n5, Alpha=1, Translation=Get ShardFloatA() -> (ShardFloatA:Vector), Rotation="0,0,0", Scale="1,1,1") {
    rotationMode=BMM_Additive
    scaleSpace=BCS_BoneSpace
    translationMode=BMM_Additive
    translationSpace=BCS_BoneSpace
    alphaBoolBlend {blendOption=Linear}
    boneToModify {boneName=dyn_shard_r}
   }
   n7 = "Transform (Modify) Bone - Bone: dyn_shard_l"(ComponentPose=@n6, Alpha=1, Translation=Get ShardFloatA(), Rotation="0,0,0", Scale="1,1,1") {
    rotationMode=BMM_Additive
    scaleSpace=BCS_BoneSpace
    translationMode=BMM_Additive
    translationSpace=BCS_BoneSpace
    alphaBoolBlend {blendOption=Linear}
    boneToModify {boneName=dyn_shard_l}
   }
   n8 = "Transform (Modify) Bone - Bone: dyn_shard_m"(ComponentPose=@n7, Alpha=1, Translation=Get ShardFloatC() -> (ShardFloatC:Vector), Rotation="0,0,0", Scale="1,1,1") {
    rotationMode=BMM_Additive
    scaleSpace=BCS_BoneSpace
    translationMode=BMM_Additive
    translationSpace=BCS_BoneSpace
    alphaBoolBlend {blendOption=Linear}
    boneToModify {boneName=dyn_shard_m}
   }
   n9 = "Constraint - dyn_hair_l_back_01 (PP)"!(ComponentPose=@n8, Alpha=1, ConstraintWeights_0=0.5, ConstraintWeights_1:float=0.5) -> (Pose) {
    boneToModify {boneName=dyn_hair_l_back_01}
    constraintSetup [{targetBone {boneName=neck_02}}, {targetBone {boneName=head}}]
    constraintWeights [1,1]
   }
   n10 = "Constraint - dyn_hair_m_back_01 (PP)"!(ComponentPose=@n9, Alpha=1, ConstraintWeights_0=0.5, ConstraintWeights_1:float=0.5) -> (Pose) {
    boneToModify {boneName=dyn_hair_m_back_01}
    constraintSetup [{targetBone {boneName=neck_02}}, {targetBone {boneName=head}}]
    constraintWeights [1,1]
   }
   n11 = "Constraint - dyn_hair_r_back_01 (PP)"!(ComponentPose=@n10, Alpha=1, ConstraintWeights_0=0.5, ConstraintWeights_1:float=0.5) -> (Pose) {
    boneToModify {boneName=dyn_hair_r_back_01}
    constraintSetup [{targetBone {boneName=neck_02}}, {targetBone {boneName=head}}]
    constraintWeights [1,1]
   }
   n12 = "Constraint - dyn_hair_l_back_02 (PPP)"!(ComponentPose=@n11, Alpha=1, ConstraintWeights_0=0.3, ConstraintWeights_1:float=0.3, ConstraintWeights_2:float=0.4) -> (Pose) {
    boneToModify {boneName=dyn_hair_l_back_02}
    constraintSetup [{targetBone {boneName=spine_05}}, {targetBone {boneName=neck_01}}, {targetBone {boneName=head}}]
    constraintWeights [1,1,1]
   }
   n13 = "Constraint - dyn_hair_m_back_02 (PP)"!(ComponentPose=@n12, Alpha=1, ConstraintWeights_0=0.3, ConstraintWeights_1:float=0.7) -> (Pose) {
    boneToModify {boneName=dyn_hair_m_back_02}
    constraintSetup [{targetBone {boneName=spine_05}}, {targetBone {boneName=neck_01}}]
    constraintWeights [1,1]
   }
   n14 = "Constraint - dyn_hair_r_back_02 (PPP)"!(ComponentPose=@n13, Alpha=1, ConstraintWeights_0=0.3, ConstraintWeights_1:float=0.3, ConstraintWeights_2:float=0.4) -> (Pose) {
    boneToModify {boneName=dyn_hair_r_back_02}
    constraintSetup [{targetBone {boneName=spine_05}}, {targetBone {boneName=neck_01}}, {targetBone {boneName=head}}]
    constraintWeights [1,1,1]
   }
   Output Pose(Result=Component To Local(ComponentPose=@n14)) {layerGroup=None}
  ]}
 Dyn Shards {nodes [
   Dyn Shards() {is=title K2Node_FunctionEntry}
   > Set PawnSpeedRemap(PawnSpeedRemap:double=MapRangeClamped(Value=Get PawnSpeed2D(), InRangeA=0, InRangeB=549)) -> (Output_Get:double)
   n1 = GetGameTimeInSeconds()
   n2 = "Sin (Radians)"(A=float * float(A=float * float(A=1, B=1), B=@n1))
   n3 = "Sin (Radians)"(A=float * float(A=float * float(A=1, B=2), B=@n1))
   n4 = "Sin (Radians)"(A=float * float(A=float * float(A=1, B=1.5), B=@n1))
   > Set ShardFloatA(ShardFloatA:Vector=MakeVector(X=float * float(A=@n4, B=0.3), Y=float * float(A=@n2, B=0.56), Z=float * float(A=@n3, B=0.22))) -> (Output_Get:Vector)
   n5 = "Cos (Radians)"!(A:double=float * float(A=float * float(A=1, B=1.5), B=@n1)) -> (ReturnValue:double)
   n6 = "Cos (Radians)"(A=float * float(A=float * float(A=1, B=1.78), B=@n1))
   n7 = "Cos (Radians)"(A=float * float(A=float * float(A=1, B=1.15), B=@n1))
   > Set ShardFloatB(ShardFloatB:Vector=MakeVector(X=float * float(A=@n5, B=0.6), Y=float * float(A=@n6, B=0.725), Z=float * float(A=@n7, B=0.3))) -> (Output_Get:Vector)
   n8 = "Sin (Radians)"(A=float * float(A=float * float(A=1, B=1.5), B=@n1))
   n9 = "Sin (Radians)"(A=float * float(A=float * float(A=1, B=2), B=@n1))
   n10 = "Sin (Radians)"(A=float * float(A=float * float(A=1, B=1), B=@n1))
   > Set ShardFloatC(ShardFloatC:Vector=MakeVector(X=float * float(A=@n8, B=0.3), Y=float * float(A=@n10, B=0.6), Z=float * float(A=@n9, B=0.25))) -> (Output_Get:Vector)
  ]}
 EventGraph {nodes [
   Event BlueprintUpdateAnimation()
   > "Gravity Override State Function (Parameter Asset) (Nativized)"(
    CharacterPartAnimInstance=Self-Reference()
    GameTimeInSeconds=GetGameTimeInSeconds()
    Parameters=/BRCosmetics/Characters/Player/Male/Medium/Bodies/M_MED_ScotchPeek/Meshes/F_MED_ScotchPeek_Parameters
   )
   > Set Gravity Override(Gravity Override="0,0,-980")
   > Dyn Shards() {is=title K2Node_CallFunction}
  ]}
]
variables [
 Gravity Override {type=Vector}
 PawnSpeedRemap {type=number}
 ShardFloatA {type=Vector}
 ShardFloatB {type=Vector}
 ShardFloatC {type=Vector}
]
